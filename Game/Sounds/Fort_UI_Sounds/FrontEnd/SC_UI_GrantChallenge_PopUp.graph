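# Front-end UI Sound Cue for the grant challenge pop-up, assigned to the Fort_UI sound class.
# A Mixer plays two layers together: a generic toolkit medium whoosh and a "more info" modal appear sound reused from the About_Modal set.
# A Modulator holds the modal layer at a fixed reduced volume, keeping it beneath the whoosh.
# Its concurrency override values specify a single instance with new plays prevented.
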SoundCue /Game/Sounds/Fort_UI_Sounds/FrontEnd/SC_UI_GrantChallenge_PopUp
root=Mixer_1
nodes [
 {id=Mixer_1, children [WavePlayer_6,Modulator_2], properties {inputVolume [1,1]}}
 {id=Modulator_2, children [WavePlayer_5], properties {volumeMax=0.3, volumeMin=0.3}}
 {id=WavePlayer_5, properties {soundWaveAssetPtr=/Game/Sounds/Fort_UI_Sounds/FrontEnd/About_Modal/wav/UI_S24_MSE_MoreInfo_Modal_Appear}}
 {
  id=WavePlayer_6
  properties {soundWaveAssetPtr=/Game/Sounds/Fort_UI_Sounds/FrontEnd/_GenericToolkit_/Whooshes/Misc_Whooshes/wav/UI_Toolkit_Whoosh_Med_02}
 }
]
properties {
 maxDistance=2097152
 soundClassObject=&Fort_UI
 volumeMultiplier=1
 concurrencyOverrides {maxCount=1, resolutionRule=PreventNew}
}
