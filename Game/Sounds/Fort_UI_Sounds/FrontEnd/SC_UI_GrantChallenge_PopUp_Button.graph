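# Sound cue for the front-end grant challenge pop-up button in the Fort_UI sound class.
# A wave player plays the gift-received prompt sound, borrowed from the rewards and gifts folder, through a modulator at default settings.
# Concurrency allows one instance at a time and blocks new plays while it is active, with volume slightly reduced.
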SoundCue /Game/Sounds/Fort_UI_Sounds/FrontEnd/SC_UI_GrantChallenge_PopUp_Button
root=Modulator_1
nodes [
 {id=Modulator_1, children [WavePlayer_2]}
 {id=WavePlayer_2, properties {soundWaveAssetPtr=/Game/Sounds/Fort_UI_Sounds/FrontEnd/RewardsAndGifts/wav/UI_Gift_Received_Prompt}}
]
properties {
 maxDistance=2097152
 soundClassObject=&Fort_UI
 volumeMultiplier=0.9
 concurrencyOverrides {maxCount=1, resolutionRule=PreventNew}
}
