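# Modal widget on AthenaGrantChallengeModal that announces a granted challenge bundle with name, context, description, reward image and confirm button.
# OnChallengeBundleSet runs Set Styling, which plays Intro and applies the challenge set's colours to dynamic materials and text, plus lazy-loaded images.
# Clicking disables the button and plays Intro in reverse; on finish it unbinds, re-enables and shows the next popup.
# Whoosh sounds come from an animation-started binding and a sequencer event function.
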WidgetBlueprint /Gifting/Items/Messages/GrantChallenge/GrantChallenge
animations [Intro {
  binding_count=9
  length_seconds=1.2
  object=&self:Intro
  sequence {
   bindings [
    GlowBackground {
     id=#1
     possessed_class=Image
     tracks [
      {
       type=MovieSceneFloatTrack
       display_name=Render Opacity
       sections [{type=MovieSceneFloatSection, channels [None {type=MovieSceneScriptingFloatChannel, keys [[1,0],[2,1],[4,1]]}]}]
      }
      {
       type=MovieScene2DTransformTrack
       sections [{
         type=MovieScene2DTransformSection
         channels [
          Translation.X {type=MovieSceneScriptingFloatChannel, const [0,2]}
          Translation.Y {type=MovieSceneScriptingFloatChannel, const [0,2]}
          Angle {type=MovieSceneScriptingFloatChannel, const [0,3]}
          Scale.X {type=MovieSceneScriptingFloatChannel}
          Scale.Y {type=MovieSceneScriptingFloatChannel}
          Shear.X {type=MovieSceneScriptingFloatChannel, const [0,2]}
          Shear.Y {type=MovieSceneScriptingFloatChannel, const [0,2]}
         ]
        }]
      }
      {
       type=MovieSceneWidgetMaterialTrack
       sections [{
         type=MovieSceneParameterSection
         channels [
          HotsportEdge {type=MovieSceneScriptingFloatChannel, keys [[2,2],[8,3],[16,2]]}
          Density {type=MovieSceneScriptingFloatChannel, keys [[2,0.5],[8,1.5],[16,1.5]]}
          HotspotSize {type=MovieSceneScriptingFloatChannel, keys [[2,1],[8,0.5],[16,0.5]]}
         ]
        }]
      }
     ]
    }
    SizeBoxButtonSize {
     id=#2
     possessed_class=SizeBox
     tracks [
      {
       type=MovieScene2DTransformTrack
       sections [{
         type=MovieScene2DTransformSection
         channels [
          Translation.X {type=MovieSceneScriptingFloatChannel}
          Translation.Y {type=MovieSceneScriptingFloatChannel}
          Angle {type=MovieSceneScriptingFloatChannel}
          Scale.X {type=MovieSceneScriptingFloatChannel, keys [[21.0003,2],[23.0003,1],[24.0003,1]]}
          Scale.Y {type=MovieSceneScriptingFloatChannel, keys [[21.0003,2],[23.0003,1],[24.0003,1]]}
          Shear.X {type=MovieSceneScriptingFloatChannel}
          Shear.Y {type=MovieSceneScriptingFloatChannel}
         ]
        }]
      }
      {
       type=MovieSceneFloatTrack
       display_name=Render Opacity
       sections [{type=MovieSceneFloatSection, channels [None {type=MovieSceneScriptingFloatChannel, keys [[21.0003,0],[22.0003,1]]}]}]
      }
      {
       type=MovieSceneByteTrack
       display_name=Visibility
       sections [{type=MovieSceneByteSection, channels [None {type=MovieSceneScriptingByteChannel, keys [[20,1],[21,4]]}]}]
      }
     ]
    }
    SizeBox_0 {
     id=#3
     possessed_class=SizeBox
     tracks [
      {
       type=MovieScene2DTransformTrack
       sections [{
         type=MovieScene2DTransformSection
         channels [
          Translation.X {type=MovieSceneScriptingFloatChannel, const [0,2]}
          Translation.Y {type=MovieSceneScriptingFloatChannel, keys [[3,256],[6.002,0],[7.002,0]]}
          Angle {type=MovieSceneScriptingFloatChannel}
          Scale.X {type=MovieSceneScriptingFloatChannel, keys [[3,0.75],[5,1],[6.002,1]]}
          Scale.Y {type=MovieSceneScriptingFloatChannel, thinned_from=4, keys [[3,1.25],[5,1],[7.002,1]]}
          Shear.X {type=MovieSceneScriptingFloatChannel, const [0,2]}
          Shear.Y {type=MovieSceneScriptingFloatChannel, const [0,3]}
         ]
        }]
      }
      {
       type=MovieSceneFloatTrack
       display_name=Render Opacity
       sections [{type=MovieSceneFloatSection, channels [None {type=MovieSceneScriptingFloatChannel, keys [[3,0],[5,1]]}]}]
      }
     ]
    }
    Backplate {
     id=#4
     possessed_class=Image
     tracks [{
       type=MovieSceneFloatTrack
       display_name=Render Opacity
       sections [{type=MovieSceneFloatSection, channels [None {type=MovieSceneScriptingFloatChannel, keys [[2,0],[4,1]]}]}]
      }]
    }
    Text_BodyCopy {
     id=#5
     possessed_class=CommonTextBlock
     tracks [
      {
       type=MovieSceneFloatTrack
       display_name=Render Opacity
       sections [{type=MovieSceneFloatSection, channels [None {type=MovieSceneScriptingFloatChannel, keys [[9,0],[11,1]]}]}]
      }
      {
       type=MovieScene2DTransformTrack
       sections [{
         type=MovieScene2DTransformSection
         channels [
          Translation.X {type=MovieSceneScriptingFloatChannel}
          Translation.Y {type=MovieSceneScriptingFloatChannel}
          Angle {type=MovieSceneScriptingFloatChannel}
          Scale.X {type=MovieSceneScriptingFloatChannel}
          Scale.Y {type=MovieSceneScriptingFloatChannel}
          Shear.X {type=MovieSceneScriptingFloatChannel}
          Shear.Y {type=MovieSceneScriptingFloatChannel}
         ]
        }]
      }
     ]
    }
    LazyImage_RewardPreview {
     id=#6
     possessed_class=FortLazyImage
     tracks [
      {
       type=MovieSceneFloatTrack
       display_name=Render Opacity
       sections [{type=MovieSceneFloatSection, channels [None {type=MovieSceneScriptingFloatChannel, keys [[4,0],[7,1]]}]}]
      }
      {
       type=MovieScene2DTransformTrack
       sections [{
         type=MovieScene2DTransformSection
         channels [
          Translation.X {type=MovieSceneScriptingFloatChannel}
          Translation.Y {type=MovieSceneScriptingFloatChannel}
          Angle {type=MovieSceneScriptingFloatChannel}
          Scale.X {type=MovieSceneScriptingFloatChannel, keys [[4,0.125],[7,1],[8,1]]}
          Scale.Y {type=MovieSceneScriptingFloatChannel, keys [[4,0.125],[7,1],[8,1]]}
          Shear.X {type=MovieSceneScriptingFloatChannel}
          Shear.Y {type=MovieSceneScriptingFloatChannel}
         ]
        }]
      }
     ]
    }
    Text_ChallengeName {
     id=#7
     possessed_class=CommonTextBlock
     tracks [
      {
       type=MovieScene2DTransformTrack
       sections [{
         type=MovieScene2DTransformSection
         channels [
          Translation.X {type=MovieSceneScriptingFloatChannel, const [0,3]}
          Translation.Y {type=MovieSceneScriptingFloatChannel, keys [[5,64],[8,0],[10,0]]}
          Angle {type=MovieSceneScriptingFloatChannel}
          Scale.X {type=MovieSceneScriptingFloatChannel, keys [[5,0.5],[8,1],[10,1]]}
          Scale.Y {type=MovieSceneScriptingFloatChannel}
          Shear.X {type=MovieSceneScriptingFloatChannel}
          Shear.Y {type=MovieSceneScriptingFloatChannel}
         ]
        }]
      }
      {
       type=MovieSceneFloatTrack
       display_name=Render Opacity
       sections [{type=MovieSceneFloatSection, channels [None {type=MovieSceneScriptingFloatChannel, keys [[5,0],[8,1]]}]}]
      }
     ]
    }
    Text_ChallengeContext {
     id=#8
     possessed_class=CommonTextBlock
     tracks [
      {
       type=MovieScene2DTransformTrack
       sections [{
         type=MovieScene2DTransformSection
         channels [
          Translation.X {type=MovieSceneScriptingFloatChannel}
          Translation.Y {type=MovieSceneScriptingFloatChannel}
          Angle {type=MovieSceneScriptingFloatChannel}
          Scale.X {type=MovieSceneScriptingFloatChannel, keys [[7,0.5],[9,1]]}
          Scale.Y {type=MovieSceneScriptingFloatChannel, const [1,2]}
          Shear.X {type=MovieSceneScriptingFloatChannel}
          Shear.Y {type=MovieSceneScriptingFloatChannel}
         ]
        }]
      }
      {
       type=MovieSceneFloatTrack
       display_name=Render Opacity
       sections [{type=MovieSceneFloatSection, channels [None {type=MovieSceneScriptingFloatChannel, keys [[8,0],[10,1]]}]}]
      }
     ]
    }
    BackgroundBlur_46 {
     id=#9
     possessed_class=BackgroundBlur
     tracks [{
       type=MovieSceneFloatTrack
       display_name=Blur Strength
       sections [{type=MovieSceneFloatSection, channels [None {type=MovieSceneScriptingFloatChannel, keys [[0,0],[2,20]]}]}]
      }]
    }
   ]
   tracks [
    {type=MovieSceneAudioTrack, display_name=Audio, sections []}
    {
     type=MovieSceneEventTrack
     sections [{
       type=MovieSceneEventTriggerSection
       channels [None {
         type=MovieSceneScriptingEventChannel
         keys [[0, "<Struct 'MovieSceneEvent' (0x467a51a87a0) {}>"], [19, "<Struct 'MovieSceneEvent' (0x467a51a8660) {}>"]]
        }]
      }]
    }
   ]
  }
 }]
graphs [
 EventGraph {nodes [
   "On Button Clicked (IconAction)"!() -> (OutputDelegate, then, Button:Common Button Legacy Object Reference) {is=title K2Node_ComponentBoundEvent}
   > n1 = SetIsEnabled(self=Get IconAction() -> (IconAction:Icon Text Button Object Reference), bInIsEnabled=false)
   n2 = OnAnimFinished() {is=title K2Node_CustomEvent}
   n3 = Get Intro()
   BindToAnimationFinished(execute=@n1.then, Animation=@n3, Delegate=@n2)
   > PlayAnimation(InAnimation=@n3, PlayMode=Reverse)
   n4 = Play Intro Audio() {is=title K2Node_CustomEvent}
   > PlaySound2D(Sound=Get IntroWhoosh() -> (IntroWhoosh:Sound Base Object Reference))
   n5 = Event OnChallengeBundleSet!() -> (OutputDelegate, then, DisplayStyle:"Fort Challenge Set Style Structure (by ref)") {is=title K2Node_Event}
   > Set Styling!(execute, self, FortChallengeSetStyle:"Fort Challenge Set Style Structure (by ref)"=@n5.DisplayStyle) -> (then) {is=title K2Node_CallFunction}
   Event Construct()
   > BindToAnimationStarted!(execute, self:User Widget Object Reference, Animation:Widget Animation Object Reference=Get Intro(), Delegate:Delegate=@n4) -> (then)
   Break Fort Challenge Set Style!(FortChallengeSetStyle:"Fort Challenge Set Style Structure (by ref)") -> (
    PrimaryColor:Linear Color Structure
    SecondaryColor:Linear Color Structure
    AccentColor:Linear Color Structure
    Context_LimitedTimeColor:Linear Color Structure
    Context_BaseColor:Linear Color Structure
    DisplayImage:Soft Object Reference
    CustomBackground:Soft Object Reference
   ) {is=title K2Node_BreakStruct}
   ShowNextChallengePopup(execute=@n6.then, self:Athena Grant Challenge Modal Object Reference)
   UnbindAllFromAnimationFinished!(execute=@n2.then, self:User Widget Object Reference, Animation:Widget Animation Object Reference=@n3) -> (then)
   > n6 = SetIsEnabled(self=Get IconAction(), bInIsEnabled=true)
  ]}
 Sequencer Events {nodes [
   ButtonWhoosh() {is=title K2Node_CustomEvent}
   > SpawnSound2D!(
    execute
    Sound:Sound Base Object Reference=Get ButtonWhooshSound() -> (ButtonWhooshSound:Sound Base Object Reference)
    VolumeMultiplier:float=1
    PitchMultiplier:float=1
    StartTime:float=0
    ConcurrencySettings:Sound Concurrency Object Reference
    bPersistAcrossLevelTransition:bool=false
    bAutoDestroy:bool=true
   ) -> (then, ReturnValue:Audio Component Object Reference)
  ]}
 Set Styling {nodes [
   Set Styling!() -> (then, FortChallengeSetStyle:"Fort Challenge Set Style Structure (by ref)") {is=title K2Node_FunctionEntry}
   > PlayAnimation(InAnimation=Get Intro())
   > n1 = GetDynamicMaterial(self=Get GlowBackground() -> (GlowBackground:Image Object Reference))
   n2 = Break Fort Challenge Set Style!(FortChallengeSetStyle=Get FortChallengeSetStyle() -> (FortChallengeSetStyle:"Fort Challenge Set Style Structure (by ref)")) -> (PrimaryColor, SecondaryColor)
   > SetVectorParameterValue(self=@n1, ParameterName=Primary, Value=@n2)
   > SetVectorParameterValue(self=@n1, ParameterName=Secondary, Value=@n2.SecondaryColor)
   > n3 = GetDynamicMaterial(self=Get Image_Background() -> (Image_Background:Image Object Reference))
   n4 = Break Fort Challenge Set Style!(FortChallengeSetStyle=Get FortChallengeSetStyle()) -> (
    PrimaryColor
    SecondaryColor
    AccentColor:Linear Color Structure
    Context_LimitedTimeColor:Linear Color Structure
    Context_BaseColor:Linear Color Structure
    DisplayImage:Soft Object Reference
    CustomBackground:Soft Object Reference
   )
   > SetVectorParameterValue(self=@n3, ParameterName=Primary, Value=@n4)
   > SetVectorParameterValue(self=@n3, ParameterName=Secondary, Value=@n4.SecondaryColor)
   n5 = Break Fort Challenge Set Style!(FortChallengeSetStyle=Get FortChallengeSetStyle()) -> (CustomBackground)
   > SetBrushFromLazyDisplayAsset!(
    execute
    self:Common Lazy Image Object Reference=Get LazyImage_CustomBackground() -> (LazyImage_CustomBackground:Fort Lazy Image Object Reference)
    LazyObject:"Soft Object Reference (by ref)"=@n5
    bMatchTextureSize:bool=false
   ) -> (then)
   > SetVisibility(
    self=Get LazyImage_CustomBackground()
    InVisibility=Select(Option 0=Collapsed, Option 1=Visible, Index=IsValidSoftObjectReference(SoftObjectReference=@n5))
   )
   > SetBrushFromLazyDisplayAsset(
    self=Get LazyImage_RewardPreview() -> (LazyImage_RewardPreview:Fort Lazy Image Object Reference)
    LazyObject=Break Fort Challenge Set Style!(FortChallengeSetStyle=Get FortChallengeSetStyle()) -> (DisplayImage:Soft Object Reference)
    bMatchTextureSize=false
   )
   > n6 = GetDynamicMaterial(self=Get Image_Border() -> (Image_Border:Image Object Reference))
   n7 = Break Fort Challenge Set Style!(FortChallengeSetStyle=Get FortChallengeSetStyle()) -> (
    PrimaryColor:Linear Color Structure
    SecondaryColor:Linear Color Structure
    AccentColor:Linear Color Structure
    Context_LimitedTimeColor:Linear Color Structure
    Context_BaseColor:Linear Color Structure
    DisplayImage
    CustomBackground:Soft Object Reference
   )
   > SetVectorParameterValue(self=@n6, ParameterName=Primary, Value=@n7)
   > SetVectorParameterValue(self=@n6, ParameterName=Secondary, Value=@n7.SecondaryColor)
   n8 = Break Fort Challenge Set Style(FortChallengeSetStyle=Get FortChallengeSetStyle())
   > SetColorAndOpacity(
    self=Get Text_ChallengeContext() -> (Text_ChallengeContext:Common Text Object Reference)
    InColorAndOpacity=Make Slate Color(SpecifiedColor=@n8.AccentColor)
   )
   > SetColorAndOpacity(
    self=Get Text_BodyCopy() -> (Text_BodyCopy:Common Text Object Reference)
    InColorAndOpacity=Make Slate Color(SpecifiedColor=@n8.Context_LimitedTimeColor)
   )
  ]}
]
variables [
 ButtonWhooshSound {type=SoundBase, default=/Game/Sounds/Fort_UI_Sounds/FrontEnd/SC_UI_GrantChallenge_PopUp_Button}
 IntroWhoosh {type=SoundBase, default=/Game/Sounds/Fort_UI_Sounds/FrontEnd/SC_UI_GrantChallenge_PopUp}
]
widgets {
 info {inheritedWidgetCount=4, parentClass=/Script/FortniteUI.AthenaGrantChallengeModal, rootWidgetClass=&Overlay, widgetCount=19}
 widgets [
  Overlay_69 {widgetClassPath=&Overlay, widget=Overlay_69}
  BackgroundBlur_46 {
   widgetClassPath=&BackgroundBlur
   parent=Overlay_69
   widget=BackgroundBlur_46
   properties {blurStrength=20}
   slot {class=OverlaySlot, properties {horizontalAlignment=HAlign_Fill, verticalAlignment=VAlign_Fill}}
  }
  Backplate {
   widgetClassPath=&Image
   bIsVariable=true
   parent=Overlay_69
   widget=Backplate
   properties {colorAndOpacity=<rgba 0 0.05 0.5 0.2>}
   slot {class=OverlaySlot, properties {horizontalAlignment=HAlign_Fill, verticalAlignment=VAlign_Fill}}
  }
  GlowBackground {
   widgetClassPath=&Image
   bIsVariable=true
   parent=Overlay_69
   widget=GlowBackground
   properties {brush {
     bIsDynamicallyLoaded=false
     drawAs=Image
     imageType=NoImage
     mirroring=NoMirror
     resourceName=None
     resourceObject=/Game/Athena/UI/Frontend/GrantChallenge/M-UI-Challenge-GlowBack
     tiling=NoTile
     imageSize=<32 32>
     margin=<ltrb 0 0 0 0>
     outlineSettings {
      bUseBrushTransparency=false
      roundingType=HalfHeightRadius
      width=0
      color {colorUseRule=UseColor_Specified, specifiedColor=<rgba 0 0 0 0>}
      cornerRadii=<0 0 0 0>
     }
     tintColor {colorUseRule=UseColor_Specified, specifiedColor=<rgba 1 1 1 1>}
     uVRegion {bIsValid=false, max=<0 0>, min=<0 0>}
    }}
   slot {class=OverlaySlot, properties {horizontalAlignment=HAlign_Fill, verticalAlignment=VAlign_Fill}}
  }
  SizeBox_0 {
   widgetClassPath=&SizeBox
   parent=Overlay_69
   widget=SizeBox_0
   properties {
    bOverride_MinDesiredHeight=true
    bOverride_WidthOverride=true
    heightOverride=512
    maxDesiredWidth=512
    minDesiredHeight=620
    minDesiredWidth=512
    widthOverride=620
   }
   slot {class=OverlaySlot, properties {horizontalAlignment=HAlign_Center, verticalAlignment=VAlign_Center}}
  }
  Overlay_112 {widgetClassPath=&Overlay, parent=SizeBox_0, widget=Overlay_112, slot {class=SizeBoxSlot}}
  Image_Background {
   widgetClassPath=&Image
   bIsVariable=true
   parent=Overlay_112
   widget=Image_Background
   properties {brush {
     bIsDynamicallyLoaded=false
     drawAs=Image
     imageType=NoImage
     mirroring=NoMirror
     resourceName=None
     resourceObject=/Game/Athena/UI/Frontend/GrantChallenge/M-UI-Challene-Backplate
     tiling=NoTile
     imageSize=<512 512>
     margin=<ltrb 0 0 0 0>
     outlineSettings {
      bUseBrushTransparency=false
      roundingType=HalfHeightRadius
      width=0
      color {colorUseRule=UseColor_Specified, specifiedColor=<rgba 0 0 0 0>}
      cornerRadii=<0 0 0 0>
     }
     tintColor {colorUseRule=UseColor_Specified, specifiedColor=<rgba 1 1 1 1>}
     uVRegion {bIsValid=false, max=<0 0>, min=<0 0>}
    }}
   slot {class=OverlaySlot, properties {horizontalAlignment=HAlign_Fill, verticalAlignment=VAlign_Fill}}
  }
  LazyImage_CustomBackground {
   widgetClassPath=&FortLazyImage
   bIsVariable=true
   parent=Overlay_112
   widget=LazyImage_CustomBackground
   properties {
    materialTextureParamName=Texture
    brush {
     bIsDynamicallyLoaded=false
     drawAs=Image
     imageType=NoImage
     mirroring=NoMirror
     resourceName=None
     resourceObject=/Game/Athena/UI/Frontend/GrantChallenge/M-UI-Custom-Background
     tiling=NoTile
     imageSize=<32 32>
     margin=<ltrb 0 0 0 0>
     outlineSettings {
      bUseBrushTransparency=false
      roundingType=HalfHeightRadius
      width=0
      color {colorUseRule=UseColor_Specified, specifiedColor=<rgba 0 0 0 0>}
      cornerRadii=<0 0 0 0>
     }
     tintColor {colorUseRule=UseColor_Specified, specifiedColor=<rgba 1 1 1 1>}
     uVRegion {bIsValid=false, max=<0 0>, min=<0 0>}
    }
   }
   slot {class=OverlaySlot, properties {horizontalAlignment=HAlign_Fill, verticalAlignment=VAlign_Fill}}
  }
  LazyImage_RewardPreview {
   widgetClassPath=&FortLazyImage
   bIsVariable=true
   parent=Overlay_112
   widget=LazyImage_RewardPreview
   properties {
    materialTextureParamName=DisplayImage
    brush {
     bIsDynamicallyLoaded=false
     drawAs=Image
     imageType=FullColor
     mirroring=NoMirror
     resourceName=None
     resourceObject=/Gifting/UI/Art/T-Variant-M-VigilanteBoard-BackpackHalloween-L
     tiling=NoTile
     imageSize=<512 512>
     margin=<ltrb 0 0 0 0>
     outlineSettings {
      bUseBrushTransparency=false
      roundingType=HalfHeightRadius
      width=0
      color {colorUseRule=UseColor_Specified, specifiedColor=<rgba 0 0 0 0>}
      cornerRadii=<0 0 0 0>
     }
     tintColor {colorUseRule=UseColor_Specified, specifiedColor=<rgba 1 1 1 1>}
     uVRegion {bIsValid=false, max=<0 0>, min=<0 0>}
    }
    renderTransformPivot=<0.5 1>
   }
   slot {class=OverlaySlot, properties {horizontalAlignment=HAlign_Center, padding=<ltrb 0 -128 0 0>}}
  }
  Overlay_0 {
   widgetClassPath=&Overlay
   parent=Overlay_112
   widget=Overlay_0
   slot {class=OverlaySlot, properties {horizontalAlignment=HAlign_Fill, verticalAlignment=VAlign_Fill, padding=<ltrb 0 280 0 0>}}
  }
  Image_Border {
   widgetClassPath=&Image
   bIsVariable=true
   parent=Overlay_0
   widget=Image_Border
   properties {
    visibility=HitTestInvisible
    brush {
     bIsDynamicallyLoaded=false
     drawAs=Image
     imageType=NoImage
     mirroring=NoMirror
     resourceName=None
     resourceObject=/Game/Athena/UI/Frontend/GrantChallenge/M-UI-GrantChallengeTextBorder
     tiling=NoTile
     imageSize=<32 32>
     margin=<ltrb 0 0.5 0 0.5>
     outlineSettings {
      bUseBrushTransparency=false
      roundingType=HalfHeightRadius
      width=0
      color {colorUseRule=UseColor_Specified, specifiedColor=<rgba 0 0 0 0>}
      cornerRadii=<0 0 0 0>
     }
     tintColor {colorUseRule=UseColor_Specified, specifiedColor=<rgba 1 1 1 1>}
     uVRegion {bIsValid=false, max=<0 0>, min=<0 0>}
    }
   }
   slot {class=OverlaySlot, properties {horizontalAlignment=HAlign_Fill, verticalAlignment=VAlign_Fill}}
  }
  VerticalBox_0 {
   widgetClassPath=&VerticalBox
   parent=Overlay_0
   widget=VerticalBox_0
   slot {class=OverlaySlot, properties {horizontalAlignment=HAlign_Fill, verticalAlignment=VAlign_Fill, padding=<ltrb 45 45 45 90>}}
  }
  ScaleBox_1 {
   widgetClassPath=&ScaleBox
   parent=VerticalBox_0
   widget=ScaleBox_1
   properties {stretchDirection=DownOnly}
   slot {class=VerticalBoxSlot}
  }
  Text_ChallengeName {
   widgetClassPath=&CommonTextBlock
   bInherited=true
   bIsVariable=true
   parent=ScaleBox_1
   widget=Text_ChallengeName
   properties {
    justification=Center
    text=ALTER EGO
    textTransformPolicy=ToUpper
    font {
     bForceMonospaced=false
     bMaterialIsStencil=false
     fontMaterial=null
     fontObject=&Burbank
     letterSpacing=0
     monospacedWidth=1
     size=72
     skewAmount=0
     typefaceFontName=Burbank-Black
     outlineSettings {
      bApplyOutlineToDropShadows=false
      bMiteredCorners=false
      bSeparateFillAlpha=false
      outlineMaterial=null
      outlineSize=0
      outlineColor=<rgba 0 0 0 1>
     }
    }
    renderTransform {angle=0, scale=<1 1>, shear=<-8 0>, translation=<0 0>}
    shadowOffset=<0 0>
   }
   slot {class=ScaleBoxSlot, properties {horizontalAlignment=HAlign_Fill}}
  }
  Text_ChallengeContext {
   widgetClassPath=&CommonTextBlock
   bInherited=true
   bIsVariable=true
   parent=VerticalBox_0
   widget=Text_ChallengeContext
   properties {
    justification=Center
    scrollStyle=/FortUILibrary/Styles/Typography/Foundation/TextScrollStyle-NoFade.TextScrollStyle-NoFade_C
    text=SPECIAL EVENT
    textTransformPolicy=ToUpper
    colorAndOpacity {colorUseRule=UseColor_Specified, specifiedColor=<rgba 1 0.855 0.0331 1>}
    font {
     bForceMonospaced=false
     bMaterialIsStencil=false
     fontMaterial=null
     fontObject=&BurbankSmall
     letterSpacing=0
     monospacedWidth=1
     size=22
     skewAmount=0
     typefaceFontName=Black
     outlineSettings {
      bApplyOutlineToDropShadows=false
      bMiteredCorners=false
      bSeparateFillAlpha=false
      outlineMaterial=null
      outlineSize=0
      outlineColor=<rgba 0 0 0 1>
     }
    }
    renderTransform {angle=0, scale=<1 1>, shear=<-12 0>, translation=<0 0>}
    shadowOffset=<0 0>
   }
   slot {class=VerticalBoxSlot, properties {padding=<ltrb 0 -16 0 4>}}
  }
  Text_BodyCopy {
   widgetClassPath=&CommonTextBlock
   bInherited=true
   bIsVariable=true
   parent=VerticalBox_0
   widget=Text_BodyCopy
   properties {
    autoWrapText=true
    justification=Center
    text="{Challenge description body copy.}"
    textTransformPolicy=ToUpper
    colorAndOpacity {colorUseRule=UseColor_Specified, specifiedColor=<rgba 0.9647 1 0.2122 1>}
    font {
     bForceMonospaced=false
     bMaterialIsStencil=false
     fontMaterial=null
     fontObject=&BurbankSmall
     letterSpacing=0
     monospacedWidth=1
     size=15
     skewAmount=0
     typefaceFontName=Bold
     outlineSettings {
      bApplyOutlineToDropShadows=false
      bMiteredCorners=false
      bSeparateFillAlpha=false
      outlineMaterial=null
      outlineSize=0
      outlineColor=<rgba 0 0 0 1>
     }
    }
    shadowOffset=<0 0>
   }
   slot {class=VerticalBoxSlot, properties {verticalAlignment=VAlign_Center, size {sizeRule=Fill, value=1}}}
  }
  SizeBoxButtonSize {
   widgetClassPath=&SizeBox
   bIsVariable=true
   parent=Overlay_112
   widget=SizeBoxButtonSize
   properties {bOverride_MinDesiredWidth=true, minDesiredWidth=304}
   slot {class=OverlaySlot, properties {horizontalAlignment=HAlign_Center, verticalAlignment=VAlign_Bottom, padding=<ltrb 0 0 0 -32>}}
  }
  IconAction {
   widgetClassPath=&IconTextButton
   bIsVariable=true
   parent=SizeBoxButtonSize
   widget=IconAction
   properties {
    buttonText="NICE!"
    contentAlignment=HAlign_Center
    onButtonBaseClicked=(&self:WidgetTree.IconAction.HandleOnButtonClicked)
    onButtonBaseDoubleClicked=(&self:WidgetTree.IconAction.HandleOnButtonDoubleClicked)
    onButtonBaseHovered=(&self:WidgetTree.IconAction.HandleOnButtonHovered)
    onButtonBaseUnhovered=(&self:WidgetTree.IconAction.HandleOnButtonUnhovered)
    onSelectedChangedBase=(&self:WidgetTree.IconAction.HandleOnSelectedChanged)
    style=/FortUILibrary/Foundation/Buttons/ButtonStyles/ButtonStyle_Feature_L_Yellow_s11.ButtonStyle_Feature_L_Yellow_s11_C
    triggeringInputAction {dataTable=&InputActionDataTable, rowName=GenericForward}
   }
   slot {class=SizeBoxSlot}
  }
  GiftBoxImage {widgetClassPath=/Script/CommonUI.CommonLazyImage, bInherited=true, widget=null}
 ]
}
